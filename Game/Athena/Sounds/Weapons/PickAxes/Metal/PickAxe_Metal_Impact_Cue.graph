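# Sound cue for pickaxe impacts on metal, limited per owner to two instances, stopping the oldest.
# Its full mix layers random pickaxe hits, pitch-varied big metal impacts with their own attenuation, and quieter distant metal impacts faded in through a distance crossfade.
# A quality-level root uses this mix for its first two branches and a single pitch-modulated pickaxe wave for each of the other two.
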SoundCue /Game/Athena/Sounds/Weapons/PickAxes/Metal/PickAxe_Metal_Impact_Cue.Pickaxe_Metal_Impact_Cue
root=QualityLevel_0
nodes [
 {
  id=Attenuation_4
  children [Modulator_4]
  properties {attenuationSettings=/Game/Sounds/Attenuation/Fort_Demolition_Crossfade_Global_Hit_Medium_Att, attenuationOverrides {falloffDistance=1600}}
 }
 {
  id=Attenuation_5
  children [Random_4]
  properties {
   attenuationSettings=/Game/Sounds/Attenuation/Fort_Demolition_Crossfade_Global_Hit_Far_Att
   attenuationOverrides {falloffDistance=3000, attenuationShapeExtents {x=1000}}
  }
 }
 {
  id=DistanceCrossFade_1
  children [Attenuation_5, null]
  properties {crossFadeInput [
    {fadeInDistanceEnd=1000, fadeInDistanceStart=500, fadeOutDistanceEnd=10000, fadeOutDistanceStart=2000, volume=1}
    {fadeInDistanceEnd=0, fadeInDistanceStart=0, fadeOutDistanceEnd=0, fadeOutDistanceStart=0, volume=1}
   ]}
 }
 {id=Mixer_2, children [Modulator_1, Attenuation_4], properties {inputVolume [1,1]}}
 {id=Mixer_3, children [Mixer_2, DistanceCrossFade_1], properties {inputVolume [1,0.7]}}
 {id=Modulator_0, children [WavePlayer_0], properties {pitchMax=1.1, pitchMin=1, volumeMax=0.7, volumeMin=0.7}}
 {id=Modulator_1, children [Random_1], properties {pitchMax=1, pitchMin=1, volumeMax=0.7, volumeMin=0.7}}
 {id=Modulator_3, children [WavePlayer_3], properties {pitchMax=1.1, pitchMin=1, volumeMax=0.7, volumeMin=0.7}}
 {id=Modulator_4, children [Random_3], properties {pitchMax=1.8, pitchMin=1, volumeMax=1.3, volumeMin=1.3}}
 {id=QualityLevel_0, children [Mixer_3,Mixer_3,Modulator_0,Modulator_3]}
 {id=Random_1, children [WavePlayer_5,WavePlayer_6,WavePlayer_7,WavePlayer_8,WavePlayer_9], properties {weights [1,1,1,1,1]}}
 {id=Random_3, children [WavePlayer_16, WavePlayer_17, WavePlayer_18], properties {weights [1,1,1]}}
 {id=Random_4, children [WavePlayer_19, WavePlayer_20, WavePlayer_21], properties {weights [1,1,1]}}
 {id=WavePlayer_0, properties {soundWaveAssetPtr=/Game/Sounds/Fort_Weapons/PickAxe/PickAxe_Enemy_001}}
 {id=WavePlayer_16, properties {soundWaveAssetPtr=/Game/Sounds/Fort_Impact_Sounds/Demolition/Metal/Metal_Big_01}}
 {id=WavePlayer_17, properties {soundWaveAssetPtr=/Game/Sounds/Fort_Impact_Sounds/Demolition/Metal/Metal_Big_03}}
 {id=WavePlayer_18, properties {soundWaveAssetPtr=/Game/Sounds/Fort_Impact_Sounds/Demolition/Metal/Metal_Big_04}}
 {id=WavePlayer_19, properties {soundWaveAssetPtr=/Game/Sounds/Fort_Impact_Sounds/Demolition/Defaults/Metal_Distant_Impact_01}}
 {id=WavePlayer_20, properties {soundWaveAssetPtr=/Game/Sounds/Fort_Impact_Sounds/Demolition/Defaults/Metal_Distant_Impact_03}}
 {id=WavePlayer_21, properties {soundWaveAssetPtr=/Game/Sounds/Fort_Impact_Sounds/Demolition/Defaults/Metal_Distant_Impact_04}}
 {id=WavePlayer_3, properties {soundWaveAssetPtr=&PickAxe_Enemy_001}}
 {id=WavePlayer_5, properties {soundWaveAssetPtr=&PickAxe_Enemy_001}}
 {id=WavePlayer_6, properties {soundWaveAssetPtr=/Game/Sounds/Fort_Weapons/PickAxe/PickAxe_Enemy_002}}
 {id=WavePlayer_7, properties {soundWaveAssetPtr=/Game/Sounds/Fort_Weapons/PickAxe/PickAxe_Enemy_003}}
 {id=WavePlayer_8, properties {soundWaveAssetPtr=/Game/Sounds/Fort_Weapons/PickAxe/PickAxe_Enemy_004}}
 {id=WavePlayer_9, properties {soundWaveAssetPtr=/Game/Sounds/Fort_Weapons/PickAxe/PickAxe_Enemy_005}}
]
properties {
 attenuationSettings=&Fort_Weapon_Melee_Sml_Att
 bPrimeOnLoad=true
 maxDistance=10000
 soundClassObject=&Fort_Impacts
 volumeMultiplier=0.92
 concurrencyOverrides {bLimitToOwner=true, maxCount=2, resolutionRule=StopOldest}
 concurrencySet [&Player_Melee_Weapon_Conc]
}
